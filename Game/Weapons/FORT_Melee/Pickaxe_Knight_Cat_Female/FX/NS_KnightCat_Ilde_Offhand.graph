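# Niagara idle effect for the offhand Knight Cat pickaxe, drawing spinning blade ghosts and motion blur.
# A rate-spawned mesh emitter leaves short-lived, fading ghost blades, while two single-particle sprite emitters form counter-rotating blur discs.
# User.RotationDirection_L drives the spin, and User.IndexColors picks one of three tints.
# A system Timeline value scales color and alpha, fading the layers together.
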NiagaraSystem /Game/Weapons/FORT_Melee/Pickaxe_Knight_Cat_Female/FX/NS_KnightCat_Ilde_Offhand
dependencies {
 usedDataInterfaces [
  &NiagaraDataInterfaceArrayColor
  &NiagaraDataInterfaceColorCurve
  &NiagaraDataInterfaceVectorCurve
  &NiagaraDataInterfaceCurve
  &NiagaraDataInterfaceSkeletalMesh
 ]
 usedDynamicInputs [&SetBoolByFloatComparison]
 usedModules [
  &SystemState
  &self:SystemSpawnScript.SystemScriptSource.SystemScriptGraph.NiagaraNodeAssignment_0.SetVariables_#1
  &Timeline
 ]
}
emitters {
 Blades_Ghost_Trail001 {
  emitterUpdateScript [
   &EmitterState(Life Cycle Mode=::NewEnumerator1, Loop Duration=1, UseLoopDelay=false)
   &SpawnBurst_Instantaneous(Use Spawn Probability=false, UseLoopCountLimit=false) {enabled=false}
   &SpawnRate(SpawnRate=15, Use Spawn Probability=false)
  ]
  particleSpawnScript [
   &InitializeParticle(
    Color=&MultiplyLinearColor(
     Linear Color=&SelectLinearColorFromArray(
      Array Sampling Mode=::NewEnumerator1
      Direct Array Index=$User.IndexColors
      Color Selection Array=DataInterface() {propertyValues {ColorData [<rgba 2 0 2 1>, <rgba 1 0.6 0 1>, <rgba 0 1 1.5 1>]}}
      Random Seed=?
     )
     Scale Factor=<1 1 1 2.5>
    )
    Lifetime=0.75
    Mesh Scale Mode=::NewEnumerator3
    Mesh Uniform Scale=0.5
    Position Mode=::NewEnumerator0
    Write Material Random=true
   )
   &ShapeLocation(
    MaskSpawnGroup=false
    Offset=<-10 0 -5.5>
    Offset Mode=::NewEnumerator0
    Override Seed=false
    Radius Position=0
    Ring / Disc Distribution Mode:ENiagaraSphereDistributionMode=::NewEnumerator1
    Ring Radius=10
    Rotation Angle=90
    Rotation Axis=<0 1 0>
    Rotation Coordinate Space=::Simulation
    Rotation Mode=::NewEnumerator1
    Shape Primitive=::NewEnumerator4
    Transform Order:ENiagara_TransformOrder=::NewEnumerator1
    U Position=&Multiply_Float(A=0, B=0)
    UseOwnerScale=true
   ) {enabled=false}
   &InitialMeshOrientation(
    Mesh Orientation Mode=::NewEnumerator2
    Orientation Vector=&RotateVectorByQuaternion(Vector To Rotate=$Particles.ShapeLocation.ShapeBitangent, Quaternion=<0 0 0 1>)
    Rotation Coordinate Space=::NewEnumerator0
    Use Rotation Vector=false
   ) {enabled=false}
   &SystemLocation(MaskSpawnGroup=false, Offset=<7.5 0 0>)
   &InitialMeshOrientation(
    Mesh Orientation Mode=::NewEnumerator6
    Orientation Quaternion:Quat4f=&MakeQuatFromAxisAngle(Axis=<0 0 1>, AngleInDegrees=90)
    Use Rotation Vector=false
   ) {module=InitialMeshOrientation001}
   &CameraOffset(Camera Offset Amount=-15)
   &FN_Mesh_Attributes(
    Color=&MultiplyLinearColor(
     Linear Color=&SelectLinearColorFromArray(
      Array Sampling Mode=::NewEnumerator1
      Direct Array Index=$User.IndexColors
      Color Selection Array=DataInterface() {propertyValues {ColorData [<rgba 1 0 1 1>, <rgba 1 0.6 0 1>, <rgba 0 1 1.5 1>]}}
      Random Seed=?
     )
     Scale Factor=<1 1 1 0.25>
    )
    Create Color Attribute=true
    Create Scale Attribute=true
    Lifetime=0.75
    Scale=<0.6 0.6 0.6>
    Write to Module Namespace=true
   )
  ]
  particleUpdateScript [
   &ParticleState()
   &ScaleColor(
    Linear Color Curve=DataInterface() {propertyValues {
      AlphaCurve {keys [<at 0 1>,<at 0.0741 0.9302>,<at 0.1308 0.05>,<at 0.9264 0>]}
      BlueCurve {keys [<at 0 2.9767>,<at 0.1512 1.6279>,<at 0.5423 1>,<at 0.8523 0.5052>]}
      GreenCurve {keys [<at 0 2.9767>,<at 0.1512 1.6279>,<at 0.5423 1>,<at 0.8523 0.5052>]}
      RedCurve {keys [<at 0 2.9767>,<at 0.1512 1.6279>,<at 0.5423 1>,<at 0.8523 0.5052>]}
     }}
    Scale Mode=::NewEnumerator2
   ) {enabled=false}
   &SkeletalMeshLocation(
    MaskSpawnGroup=false
    OverrideRandomSeed=false
    Skeletal Mesh=DataInterface() {propertyValues {
      DefaultMesh=null
      ExcludeBoneName=root
      OptionalComponentClass=null
      PreviewMesh=/BRCosmetics/Characters/Player/Female/Medium/Bodies/F_MED_Knight_Cat/Meshes/F_MED_Knight_Cat
      SkinningMode=SkinOnTheFly
      SoftSourceActor=null
      SourceMode=Default
      UvSetIndex=0
      WholeMeshLOD=-1
      bExcludeBone=false
      bReadDeformedGeometry=true
      bRequireCurrentFrameData=true
      ComponentTags []
      FilteredBones []
      FilteredSockets []
      MeshUserParameter {parameter {name=None, typeDefHandle {registeredTypeIndex=1}, varData []}}
      SamplingRegions []
     }}
   ) {enabled=false}
   /SetVariables(Particles.BladeOrientation:Quat4f=&MultiplyQuaternion(
     Quaternion A=&MakeQuatFromAxisAngle(
      Axis=&RotateVectorByQuaternion(Vector To Rotate=<0 2 0>, Quaternion=$Particles.MeshOrientation)
      AngleInDegrees=&Multiply_Float(
       A=$User.RotationDirection_R
       B=&Multiply_Float(A=&OneMinusFloat(Float=&ReturnNormalizedExecIndex()), B=-15)
      )
     )
     Quaternion B=$Particles.MeshOrientation
    )) {enabled=false, isSetParameters=true}
   &UpdateMeshOrientation(
    Facing Interpolation=false
    Orientation Method=::NewEnumerator1
    Override Incoming Orientation=1
    Override Quat=true
    Stabilization Mode:ENiagara_OrientVectorCount=::NewEnumerator0
    TransformFacing=false
   ) {enabled=false}
   &ScaleColor(
    Scale Alpha=&Multiply_Float(
     A=&FloatFromCurve(FloatCurve=DataInterface() {propertyValues {Curve {keys [
          <at 0 1>
          <at 0.017 0.9636>
          {arriveTangent=-1.5311, interpMode=RCIM_Cubic, leaveTangent=-1.5311, tangentMode=RCTM_User, time=0.0444, value=0.2318}
          <at 0.2891 0.0909>
          <at 0.515 0>
          <at 0.9264 0>
         ]}}})
     B=$System.Timeline.Time
    )
    Scale RGB=&VectorFromCurve(
     VectorCurve=DataInterface() {propertyValues {
       XCurve {keys [<at 0.0022 2.5>,<at 0.1015 1>,<at 0.5423 1>,<at 0.8523 0.5052>]}
       YCurve {keys [<at 0.0022 2.5>,<at 0.1015 1>,<at 0.5423 1>,<at 0.8523 0.5052>]}
       ZCurve {keys [<at 0.0022 2.5>,<at 0.1015 1>,<at 0.5423 1>,<at 0.8523 0.5052>]}
      }}
     Scale Curve=&VectorFromFloat(Value=$System.Timeline.Time)
    )
    ScaleA=true
    ScaleRGB=true
   ) {module=ScaleColor001}
   /SetVariables(Particles.FN_Mesh_Attributes.Color:LinearColor=&MultiplyLinearColor(
     Linear Color=&ScaleLinearColorByCurve(
      InitialValue=$Particles.Initial.FN_Mesh_Attributes.Color
      ScaleCurve=DataInterface() {propertyValues {
        AlphaCurve {keys [<at 0 1>,<at 0.2895 1>,<at 0.4824 0.05>,<at 0.8219 0>]}
        BlueCurve {keys [<at 0.0022 5>,<at 0.1859 1>,<at 0.5423 1>,<at 0.8523 0.5052>]}
        GreenCurve {keys [<at 0.0022 5>,<at 0.1859 1>,<at 0.5423 1>,<at 0.8523 0.5052>]}
        RedCurve {keys [<at 0.0022 5>,<at 0.1859 1>,<at 0.5423 1>,<at 0.8523 0.5052>]}
       }}
      CurveIndex=$Particles.NormalizedAge
     )
     Scale Factor=&MakeVector4_FromFloat(Float=$System.Timeline.Time)
    )) {isSetParameters=true}
   &UpdateMeshOrientation(
    Rotation Coordinate Space=::NewEnumerator0
    Rotation Rate=&Multiply_Float(
     A=$User.RotationDirection_L
     B=&ReturnNormalizedExecIndex(
      Normalized Index Scale=&Divide_Float(A=0.2, B=&ReturnNormalizedExecIndex())
      Normalized Index Includes One=false
     )
    )
    Rotation Vector=&MakeVector(X=&Multiply_Float(A=55, B=$Particles.Age), Y=0, Z=0)
   ) {module=UpdateMeshOrientation001}
  ]
  rendererClasses [&NiagaraMeshRendererProperties]
  renderers [
   {rendererClass=&NiagaraMeshRendererProperties, rendererIndex=0}
   {rendererClass=&NiagaraMeshRendererProperties, rendererIndex=1}
  ]
 }
 FN_moblur {
  emitterUpdateScript [
   &EmitterState(
    Inactive Response=::NewEnumerator1
    Life Cycle Mode=::NewEnumerator1
    Loop Behavior=::NewEnumerator1
    Loop Duration Mode=::NewEnumerator1
    UseLoopDelay=false
   )
   &SpawnBurst_Instantaneous(Spawn Count=1, Use Spawn Probability=false)
  ]
  particleSpawnScript [/Niagara/Modules/Spawn/Initialization/InitializeParticle(
    Color=&MultiplyLinearColor(
     Linear Color=&SelectLinearColorFromArray(
      Array Sampling Mode=::NewEnumerator1
      Direct Array Index=$User.IndexColors
      Color Selection Array=DataInterface() {propertyValues {ColorData [<rgba 1 0 1 1>, <rgba 1 0.6 0 1>, <rgba 0 1 1.5 1>]}}
      Random Seed=?
     )
     Scale Factor=<0.6 0.6 0.6 0.25>
    )
    Mesh Scale=<1.3 1.3 1.3>
    Module.Write Color=true
    Module.Write Lifetime=true
    Module.Write Mass=true
    Module.Write Material Random=true
    Module.Write Position=true
    Module.Write Scale=false
    Module.Write SpriteRotation=false
    Module.Write SpriteSize=true
    Position=&SimulationPosition()
    Sprite Size=&Vector2DFromFloat(Value=130)
   )]
  particleUpdateScript [
   &ParticleState(Kill Particles When Lifetime Has Elapsed=false, Let Infinitely Lived Particles Die When Emitter Deactivates=true)
   /SetVariables(Particles.Position=&ConvertVectorToPosition(Input Position=&Add_Vector(A=<0 0 0>, B=<10 0 0>), Vector Is In Absolute Large World Space:NiagaraBool=false)) {isSetParameters=true, module=SetVariables_#2}
   &CameraOffset(Camera Offset Amount=5)
   /SetVariables(
    Particles.SpriteAlignment=&RotateVectorByQuaternion(Vector To Rotate=<1 0 0>, Quaternion=<0 0 0 1>)
    Particles.SpriteFacing=<1 0.1 0>
   ) {isSetParameters=true, module=SetVariables_#3}
   &SpriteRotationRate(Rotation Rate=&Multiply_Float(A=-2000, B=$User.RotationDirection_L))
   &ScaleColor(
    Scale RGB=&VectorFromFloat(Value=&Multiply_Float(
      A=&FNTimeOfDayScalar(Night Scale=0.2, Morning Scale=0.07, Day Scale=0.7, Evening Scale=0.15)
      B=&Sine(Normalized Angle=$Emitter.Age, Period=0.1, Scale=3, Bias=0)
     ))
    ScaleA=true
    ScaleRGB=true
   ) {enabled=false}
   &ScaleColor(
    Scale Alpha=$System.Timeline.Time
    Scale RGB=&VectorFromFloat(Value=$System.Timeline.Time)
    ScaleA=true
    ScaleRGB=true
   ) {module=ScaleColor001}
  ]
  rendererClasses [&NiagaraSpriteRendererProperties]
  renderers [{rendererClass=&NiagaraSpriteRendererProperties}]
 }
 FN_moblur001 {
  emitterUpdateScript [
   &EmitterState(
    Inactive Response=::NewEnumerator1
    Life Cycle Mode=::NewEnumerator1
    Loop Behavior=::NewEnumerator1
    Loop Duration Mode=::NewEnumerator1
    UseLoopDelay=false
   )
   &SpawnBurst_Instantaneous(Spawn Count=1, Use Spawn Probability=false)
  ]
  particleSpawnScript [&InitializeParticle(
    Color=&MultiplyLinearColor(
     Linear Color=&SelectLinearColorFromArray(
      Array Sampling Mode=::NewEnumerator1
      Direct Array Index=$User.IndexColors
      Color Selection Array=DataInterface() {propertyValues {ColorData [<rgba 1 0 1 1>, <rgba 1 0.6 0 1>, <rgba 0 0.75 1.25 1>]}}
      Random Seed=?
     )
     Scale Factor=<0.15 0.15 0.15 0.1>
    )
    Mesh Scale=<1.3 1.3 1.3>
    Module.Write Color=true
    Module.Write Lifetime=true
    Module.Write Mass=true
    Module.Write Material Random=true
    Module.Write Position=true
    Module.Write Scale=false
    Module.Write SpriteRotation=false
    Module.Write SpriteSize=true
    Position=&SimulationPosition()
    Sprite Size=&Vector2DFromFloat(Value=140)
   )]
  particleUpdateScript [
   &ParticleState(Kill Particles When Lifetime Has Elapsed=false, Let Infinitely Lived Particles Die When Emitter Deactivates=true)
   /SetVariables(Particles.Position=&ConvertVectorToPosition(Input Position=&Add_Vector(A=<0 0 0>, B=<10 0 0>), Vector Is In Absolute Large World Space:NiagaraBool=false)) {isSetParameters=true, module=SetVariables_#2}
   &CameraOffset(Camera Offset Amount=5)
   /SetVariables(
    Particles.SpriteAlignment=&RotateVectorByQuaternion(Vector To Rotate=<1 0 0>, Quaternion=<0 0 0 1>)
    Particles.SpriteFacing=<1 0.1 0>
   ) {isSetParameters=true, module=SetVariables_#3}
   &SpriteRotationRate(Rotation Rate=&Multiply_Float(A=2000, B=$User.RotationDirection_L))
   &ScaleColor(
    Scale RGB=&VectorFromFloat(Value=&Multiply_Float(
      A=&FNTimeOfDayScalar(Night Scale=0.2, Morning Scale=0.07, Day Scale=0.7, Evening Scale=0.15)
      B=&Sine(Normalized Angle=$Emitter.Age, Period=0.1, Scale=3, Bias=0)
     ))
    ScaleA=true
    ScaleRGB=true
   ) {enabled=false}
   &ScaleColor(
    Scale Alpha=$System.Timeline.Time
    Scale RGB=&VectorFromFloat(Value=$System.Timeline.Time)
    ScaleA=true
    ScaleRGB=true
   ) {module=ScaleColor001}
  ]
  rendererClasses [&NiagaraSpriteRendererProperties]
  renderers [{rendererClass=&NiagaraSpriteRendererProperties}]
 }
}
summary {
 systemName=NS_KnightCat_Ilde_Offhand
 emitters [
  {emitterName=FN_moblur, rendererClasses [&NiagaraSpriteRendererProperties]}
  {emitterName=FN_moblur001, rendererClasses [&NiagaraSpriteRendererProperties]}
  {emitterName=Blades_Ghost_Trail001, rendererClasses [&NiagaraMeshRendererProperties]}
 ]
 userVariables [
  User.AlphaCurve_L {type=NiagaraFloat, defaultValue=0, description=""}
  User.bIsFrontEndPreview {type=NiagaraBool}
  User.BladeSwing {type=NiagaraBool}
  User.IndexColors {type=NiagaraInt32}
  User.Interpolate_Rate_of_Change {type=NiagaraFloat, defaultValue=7, description=""}
  User.Interpolate_WPO {type=NiagaraFloat, defaultValue=4, description=""}
  User.PlayerSpeed {type=NiagaraFloat, defaultValue=0, description=""}
  User.RotationDirection_L {type=NiagaraFloat, defaultValue=0, description=""}
  User.RotationDirection_R {type=NiagaraFloat, defaultValue=0, description=""}
 ]
}
